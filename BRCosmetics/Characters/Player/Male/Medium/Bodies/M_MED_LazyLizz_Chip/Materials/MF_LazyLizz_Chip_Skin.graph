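# Material function for the Lazy Lizz Chip character skin that adds animated, iridescent lizard-scale effects to incoming material attributes.
# A scale mask comes from either a UV2 FX map or the TA Mask input, with a skin mask taken from the M input's blue channel.
# Panning noise modulates the color, while a spectrum texture driven by inverted Fresnel adds thin-film shimmer.
# The resulting BaseColor also feeds a masked emissive boost.
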MaterialFunction /BRCosmetics/Characters/Player/Male/Medium/Bodies/M_MED_LazyLizz_Chip/Materials/MF_LazyLizz_Chip_Skin
graph {nodes [
  n1 = NamedRerouteUsage()
  n2 = NamedRerouteUsage()
  n3 = Multiply(A=NamedRerouteUsage(), B=NamedRerouteUsage())
  n4 = ScalarParameter() {defaultValue=0.12, group=06.1_Ur A Lizard Jerry, parameterName=Lazy Lizz Scale FX - Panning Base Color Noise Intensity, sortPriority=5}
  n5 = NamedRerouteUsage()
  n6 = ScalarParameter() {defaultValue=1, group=06.1_Ur A Lizard Jerry, parameterName=Lazy Lizz Scale FX - ThinFilmScale, sortPriority=15}
  n7 = Power(
   Base=NamedRerouteUsage()
   Exp=ScalarParameter() {defaultValue=0.5, group=06.1_Ur A Lizard Jerry, parameterName=Lazy Lizz Scale FX - ThinFilmWarp, sortPriority=14}
  )
  n8 = TextureSampleParameter2D(UVs=Add(A=Multiply(A=@n6, B=@n7), B=ScalarParameter() {group=06.1_Ur A Lizard Jerry, parameterName=Lazy Lizz Scale FX - ThinFilmOffset})) {
   group=06.1_Ur A Lizard Jerry
   parameterName=Lazy Lizz Scale FX - Thin Film Texture
   samplerSource=SSM_Wrap_WorldGroupSettings
   sortPriority=11
   texture=/Game/Global/Textures/Default/Thinfilm/T_ThinFilm_Spectrum_COLOR
  }
  n9 = Multiply(
   A=ScalarParameter() {defaultValue=0.5, group=06.1_Ur A Lizard Jerry, parameterName=Lazy Lizz Scale FX - ThinFilmIntensity, sortPriority=12}
   B=@n8.RGB
  )
  n10 = Power(
   Base=NamedRerouteUsage()
   Exp=ScalarParameter() {defaultValue=2, group=06.1_Ur A Lizard Jerry, parameterName=Lazy Lizz Scale FX - ThinFilmExponent, sortPriority=13}
  )
  n11 = LinearInterpolate(
   A=LinearInterpolate(A=@n2, B=Multiply(A=@n2, B=@n3), Alpha=@n4)
   B=LinearInterpolate(A=@n5, B=Add(A=@n5, B=@n9), Alpha=Saturate(None=@n10))
   Alpha=NamedRerouteUsage()
  )
  n12 = ScalarParameter() {defaultValue=0.25, group=06.1_Ur A Lizard Jerry, parameterName=Lazy Lizz Scale FX - Scale Panning Emissive Intensity, sortPriority=4}
  n13 = Add(A=@n1, B=Multiply(A=@n11, B=@n12))
  n14 = Saturate(None=Multiply(A=NamedRerouteUsage(), B=NamedRerouteUsage()))
  FunctionOutput(None=SetMaterialAttributes!(MaterialAttributes=NamedRerouteUsage(), BaseColor=@n11, EmissiveColor=LinearInterpolate(A=@n1, B=@n13, Alpha=@n14)) {attributeSetTypes [#1, #2]})
  n15 = GetMaterialAttributes(None=FunctionInput() {inputName=Material Attributes, inputType=FunctionInput_MaterialAttributes, sortPriority=1}) {attributeGetTypes [#1, #2]}
  NamedRerouteDeclaration(None=@n15.MaterialAttributes) {name=Material Attributes, nodeColor=<rgba 0.0781 0 1 1>}
  NamedRerouteDeclaration(None=OneMinus(None=FunctionInput() {inputName=Fresnel, inputType=FunctionInput_Scalar, sortPriority=3})) {name=Fresnel 1-x, nodeColor=<rgba 0 1 0.6484 1>}
  n16 = TextureSampleParameter2D(UVs=TextureCoordinate() {coordinateIndex=1}) {
   group=06.1_Ur A Lizard Jerry
   parameterName=Lazy Lizz Scales - UV2 FX Map
   samplerSource=SSM_Wrap_WorldGroupSettings
   samplerType=SAMPLERTYPE_Masks
   sortPriority=1
   texture=/BRCosmetics/Characters/Player/Male/Medium/Bodies/M_MED_LazyLizz_Chip/Textures/T_LazyLizz_Chip_UV2_FX
  }
  n17 = ComponentMask(None=FunctionInput() {bUsePreviewValueAsDefault=true, inputName=TA Mask, sortPriority=4}) {g=true}
  NamedRerouteDeclaration(None=Multiply(
    A=NamedRerouteUsage()
    B=StaticSwitchParameter(True=@n16.G, False=@n17) {defaultValue=true, group=06.1_Ur A Lizard Jerry, parameterName=Lazy Lizz Scale FX Map is UV2, sortPriority=1}
   )) {name=Scale FX Map, nodeColor=<rgba 1 0 0.4453 1>}
  NamedRerouteDeclaration(None=@n15.BaseColor) {name=BaseColor Input, nodeColor=<rgba 0 0.7656 1 1>}
  NamedRerouteDeclaration(None=@n15.EmissiveColor) {name=EmissiveColor Input, nodeColor=<rgba 1 0.9141 0 1>}
  n18 = StaticSwitchParameter(True=TextureCoordinate() {coordinateIndex=1}, False=TextureCoordinate()) {defaultValue=true, group=06.1_Ur A Lizard Jerry, parameterName=Lazy Lizz Scale FX Map is UV2, sortPriority=1}
  n19 = Multiply(A=Time()) {constB=0.1}
  n20 = ScalarParameter() {defaultValue=0.1, group=06.1_Ur A Lizard Jerry, parameterName=Lazy Lizz Scale FX - Pan Speed, sortPriority=2}
  n21 = TextureSample(UVs=Multiply(
    A=Panner(Coordinate=@n18, Time=@n19, Speed=@n20) {constCoordinate=1}
    B=ScalarParameter() {defaultValue=18, group=06.1_Ur A Lizard Jerry, parameterName=Lazy Lizz Scale FX - Pan Noise Tile, sortPriority=3}
   )) {
   samplerSource=SSM_Wrap_WorldGroupSettings
   samplerType=SAMPLERTYPE_Grayscale
   texture=/Game/Global/Textures/Default/Noise/LowResBlurredNoiseNormalized
  }
  NamedRerouteDeclaration(None=LinearInterpolate(
    A=Constant() {r=1}
    B=@n21.RGB
    Alpha=ScalarParameter() {defaultValue=1, group=06.1_Ur A Lizard Jerry, parameterName=Lazy Lizz Scale FX - Pan Noise Opacity, sortPriority=3}
   )) {name=PanNoise, nodeColor=<rgba 1 0.3281 0 1>}
  NamedRerouteDeclaration(None=ComponentMask(None=FunctionInput() {bUsePreviewValueAsDefault=true, inputName=M, sortPriority=2}) {b=true}) {name=Skin Mask, nodeColor=<rgba 0.224 0.2595 1 1>}
 ]}
