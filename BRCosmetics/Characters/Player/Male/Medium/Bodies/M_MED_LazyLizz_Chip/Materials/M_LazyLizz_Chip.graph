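# Character body material for the LazyLizz Chip cosmetic on the medium male player skeletal mesh, using material attributes with a shading model chosen in the graph.
# The shared 2023 character parent PreFX function feeds a UV2 composite function, and its outputs also supply masks and fresnel downstream.
# A static switch, on by default, routes the result through a LazyLizz Chip skin function before the shared PostFX function finishes the material.
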Material /BRCosmetics/Characters/Player/Male/Medium/Bodies/M_MED_LazyLizz_Chip/Materials/M_LazyLizz_Chip
graph {
 properties {
  bUseMaterialAttributes=true
  bUsedWithMeshDeformer=true
  bUsedWithMorphTargets=true
  bUsedWithSkeletalMesh=true
  shadingModel=MSM_FromMaterialExpression
  usedShadingModels=Default Lit | Subsurface | Clear Coat | Subsurface Profile | Cloth
 }
 nodes [
  n1 = MaterialFunctionCall() {materialFunction=&MF_CharacterParent_2023_PreFX}
  n2 = MaterialFunctionCall!(Material=@n1.Result, TechArtMask=@n1.Result) {materialFunction=/Game/Global/Materials/Character_Parent_2023/UV2Composite_Parent_2023/MF_UV2_Composite_2023}
  MaterialFunctionCall(
   Material=StaticSwitchParameter(
    True=MaterialFunctionCall!(Material Attributes=@n2.Result, M=@n1.M, Fresnel=@n1.M, TA Mask=@n1.M) {materialFunction=&/MF_LazyLizz_Chip_Skin}.Result
    False=@n2.Result
   ) {defaultValue=true, group=04_Tech Art, parameterName=Use LazyLizz Skin}
   M=@n1.M
   Pre Skinned Local Position=@n1.M
   Distance Blend=@n1.M
   Fresnel=@n1.M
  ) {materialFunction=&MF_CharacterParent_2023_PostFX}
 ]
}
