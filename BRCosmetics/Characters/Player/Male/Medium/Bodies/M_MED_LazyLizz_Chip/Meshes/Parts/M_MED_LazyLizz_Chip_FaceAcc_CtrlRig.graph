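# Control Rig for the LazyLizz Chip face accessory, built to rotate collar bones according to head direction.
# It derives head_vector from chained neck and head local transforms and sets rotate_collar to -80 degrees.
# Per-bone collapsed graphs remap the cosine between head_vector and the bone's local offset to scale that angle into a rotation.
# Both ModifyTransforms nodes receiving these rotations are unwired from execution, so no bones move.
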ControlRigBlueprint /BRCosmetics/Characters/Player/Male/Medium/Bodies/M_MED_LazyLizz_Chip/Meshes/Parts/M_MED_LazyLizz_Chip_FaceAcc_CtrlRig
graphs [
 CollapseNode_1_ContainedGraph {nodes [
   n1 = RigVMFunctionEntryNode()
   RigVMFunction_MathVectorAngle = /Script/RigVM.RigVMFunction_MathVectorAngle(
    B=/Script/RigVM.RigVMFunction_MathVectorUnit(Value=&RigUnit_GetTransform(Item=(Type=Bone,Name="collar_09"), Space=LocalSpace, bInitial=False, Item.Name=@n1.Bone).Transform.Translation) {is=struct RigVMUnitNode, template="Unit::Execute(in Value,out Result)"}.Result
    A=RigVMVariableNode() {variable=head_vector}.Value
   ) {is=struct RigVMUnitNode, template="AngleBetween::Execute(in A,in B,out Result)"}
   Cos = /Script/RigVM.RigVMFunction_MathDoubleCos(Value=@RigVMFunction_MathVectorAngle.Result) {is=struct RigVMUnitNode, template="Cos::Execute(in Value,out Result)"}
   Multiply = &RigVMFunction_MathDoubleMul(
    B=@n1.angle
    A=&RigVMFunction_MathDoubleRemap(TargetMaximum=1, TargetMinimum=0, Value=@Cos.Result, SourceMaximum=@n1.s_max, SourceMinimum=@n1.s_min).Result
   )
   RigVMFunctionReturnNode(
    quat_rot=(Rotation=(X=0,Y=0,Z=0,W=1),Translation=(X=0,Y=0,Z=0),Scale3D=(X=1,Y=1,Z=1))
    quat_rot.Rotation=&RigVMFunction_MathQuaternionFromEuler(Euler=(X=0,Y=0,Z=0), RotationOrder=XZY, Euler.Y=@Multiply.Result).Result
   )
  ]}
 CollapseNode_1_ContainedGraph {nodes [
   n1 = RigVMFunctionEntryNode()
   RigVMFunction_MathVectorAngle = &RigVMFunction_MathVectorAngle(
    B=&RigVMFunction_MathVectorUnit(Value=&RigUnit_GetTransform(Item=(Type=Bone,Name="collar_09"), Space=LocalSpace, bInitial=False, Item.Name=@n1.Bone).Transform.Translation) {template="Unit::Execute(in Value,out Result)"}.Result
    A=RigVMVariableNode() {variable=head_vector}.Value
   ) {template="AngleBetween::Execute(in A,in B,out Result)"}
   Cos = &RigVMFunction_MathDoubleCos(Value=@RigVMFunction_MathVectorAngle.Result) {template="Cos::Execute(in Value,out Result)"}
   Multiply = &RigVMFunction_MathDoubleMul(
    B=@n1.angle
    A=&RigVMFunction_MathDoubleRemap(TargetMaximum=1, TargetMinimum=0, Value=@Cos.Result, SourceMaximum=@n1.s_max, SourceMinimum=@n1.s_min).Result
   )
   RigVMFunctionReturnNode(
    quat_rot=(Rotation=(X=0,Y=0,Z=0,W=1),Translation=(X=0,Y=0,Z=0),Scale3D=(X=1,Y=1,Z=1))
    quat_rot.Rotation=&RigVMFunction_MathQuaternionFromEuler(Euler=(X=0,Y=0,Z=0), Euler.Y=@Multiply.Result).Result
   )
  ]}
 CollapseNode_1_ContainedGraph {nodes [
   n1 = RigVMFunctionEntryNode()
   RigVMFunction_MathVectorAngle = &RigVMFunction_MathVectorAngle(
    B=&RigVMFunction_MathVectorUnit(Value=&RigUnit_GetTransform(Item=(Type=Bone,Name="collar_09"), Space=LocalSpace, bInitial=False, Item.Name=@n1.Bone).Transform.Translation) {template="Unit::Execute(in Value,out Result)"}.Result
    A=RigVMVariableNode() {variable=head_vector}.Value
   ) {template="AngleBetween::Execute(in A,in B,out Result)"}
   Cos = &RigVMFunction_MathDoubleCos(Value=@RigVMFunction_MathVectorAngle.Result) {template="Cos::Execute(in Value,out Result)"}
   Multiply = &RigVMFunction_MathDoubleMul(
    B=@n1.angle
    A=&RigVMFunction_MathDoubleRemap(TargetMaximum=1, TargetMinimum=0, Value=@Cos.Result, SourceMaximum=@n1.s_max, SourceMinimum=@n1.s_min).Result
   )
   RigVMFunctionReturnNode(
    quat_rot=(Rotation=(X=0,Y=0,Z=0,W=1),Translation=(X=0,Y=0,Z=0),Scale3D=(X=1,Y=1,Z=1))
    quat_rot.Rotation=&RigVMFunction_MathQuaternionFromEuler(Euler=(X=0,Y=0,Z=0), Euler.Y=@Multiply.Result).Result
   )
  ]}
 CollapseNode_1_ContainedGraph {nodes [
   n1 = RigVMFunctionEntryNode()
   RigVMFunction_MathVectorAngle = &RigVMFunction_MathVectorAngle(
    B=&RigVMFunction_MathVectorUnit(Value=&RigUnit_GetTransform(Item=(Type=Bone,Name="BASE_dyn_L_hair_A_1"), Space=LocalSpace, bInitial=False, Item.Name=@n1.Bone).Transform.Translation) {template="Unit::Execute(in Value,out Result)"}.Result
    A=RigVMVariableNode() {variable=head_vector}.Value
   ) {template="AngleBetween::Execute(in A,in B,out Result)"}
   Cos = &RigVMFunction_MathDoubleCos(Value=@RigVMFunction_MathVectorAngle.Result) {template="Cos::Execute(in Value,out Result)"}
   Multiply = &RigVMFunction_MathDoubleMul(
    B=@n1.angle
    A=&RigVMFunction_MathDoubleRemap(TargetMaximum=1, TargetMinimum=0, Value=@Cos.Result, SourceMaximum=@n1.s_max, SourceMinimum=@n1.s_min).Result
   )
   RigVMFunctionReturnNode(
    quat_rot=(Rotation=(X=0,Y=0,Z=0,W=1),Translation=(X=0,Y=0,Z=0),Scale3D=(X=1,Y=1,Z=1))
    quat_rot.Rotation=&RigVMFunction_MathQuaternionFromEuler(Euler=(X=0,Y=0,Z=0), Euler.Y=@Multiply.Result).Result
   )
  ]}
 CollapseNode_1_ContainedGraph {nodes [
   n1 = RigVMFunctionEntryNode()
   RigVMFunction_MathVectorAngle = &RigVMFunction_MathVectorAngle(
    B=&RigVMFunction_MathVectorUnit(Value=&RigUnit_GetTransform(Item=(Type=Bone,Name="collar_09"), Space=LocalSpace, bInitial=False, Item.Name=@n1.Bone).Transform.Translation) {template="Unit::Execute(in Value,out Result)"}.Result
    A=RigVMVariableNode() {variable=head_vector}.Value
   ) {template="AngleBetween::Execute(in A,in B,out Result)"}
   Cos = &RigVMFunction_MathDoubleCos(Value=@RigVMFunction_MathVectorAngle.Result) {template="Cos::Execute(in Value,out Result)"}
   Multiply = &RigVMFunction_MathDoubleMul(
    B=@n1.angle
    A=&RigVMFunction_MathDoubleRemap(TargetMaximum=1, TargetMinimum=0, Value=@Cos.Result, SourceMaximum=@n1.s_max, SourceMinimum=@n1.s_min).Result
   )
   RigVMFunctionReturnNode(
    quat_rot=(Rotation=(X=0,Y=0,Z=0,W=1),Translation=(X=0,Y=0,Z=0),Scale3D=(X=1,Y=1,Z=1))
    quat_rot.Rotation=&RigVMFunction_MathQuaternionFromEuler(Euler=(X=0,Y=0,Z=0), Euler.Z=@Multiply.Result).Result
   )
  ]}
 CollapseNode_1_ContainedGraph {nodes [
   n1 = RigVMFunctionEntryNode()
   RigVMFunction_MathVectorAngle = &RigVMFunction_MathVectorAngle(
    B=&RigVMFunction_MathVectorUnit(Value=&RigUnit_GetTransform(Item=(Type=Bone,Name="collar_09"), Space=LocalSpace, bInitial=False, Item.Name=@n1.Bone).Transform.Translation) {template="Unit::Execute(in Value,out Result)"}.Result
    A=RigVMVariableNode() {variable=head_vector}.Value
   ) {template="AngleBetween::Execute(in A,in B,out Result)"}
   Cos = &RigVMFunction_MathDoubleCos(Value=@RigVMFunction_MathVectorAngle.Result) {template="Cos::Execute(in Value,out Result)"}
   Multiply = &RigVMFunction_MathDoubleMul(
    B=@n1.angle
    A=&RigVMFunction_MathDoubleRemap(TargetMaximum=1, TargetMinimum=0, Value=@Cos.Result, SourceMaximum=@n1.s_max, SourceMinimum=@n1.s_min).Result
   )
   RigVMFunctionReturnNode(
    quat_rot=(Rotation=(X=0,Y=0,Z=0,W=1),Translation=(X=0,Y=0,Z=0),Scale3D=(X=1,Y=1,Z=1))
    quat_rot.Rotation=&RigVMFunction_MathQuaternionFromEuler(Euler=(X=0,Y=0,Z=0), Euler.Y=@Multiply.Result).Result
   )
  ]}
 CollapseNode_1_ContainedGraph {nodes [
   n1 = RigVMFunctionEntryNode()
   RigVMFunction_MathVectorAngle = &RigVMFunction_MathVectorAngle(
    B=&RigVMFunction_MathVectorUnit(Value=&RigUnit_GetTransform(Item=(Type=Bone,Name="collar_09"), Space=LocalSpace, bInitial=False, Item.Name=@n1.Bone).Transform.Translation) {template="Unit::Execute(in Value,out Result)"}.Result
    A=RigVMVariableNode() {variable=head_vector}.Value
   ) {template="AngleBetween::Execute(in A,in B,out Result)"}
   Cos = &RigVMFunction_MathDoubleCos(Value=@RigVMFunction_MathVectorAngle.Result) {template="Cos::Execute(in Value,out Result)"}
   Multiply = RigVMAggregateNode(
    C=1
    B=@n1.angle
    A=&RigVMFunction_MathDoubleRemap(
     TargetMaximum=1
     TargetMinimum=0
     SourceMaximum=@n1.s_max
     SourceMinimum=@n1.s_min
     Value=&RigVMFunction_MathDoubleMul(B=1, A=@Cos.Result).Result
    ).Result
   )
   &RigVMFunction_MathDoubleMul(B=0.5, A=@Multiply.Result)
   RigVMFunctionReturnNode(
    quat_rot=(Rotation=(X=0,Y=0,Z=0,W=1),Translation=(X=0,Y=0,Z=0),Scale3D=(X=1,Y=1,Z=1))
    quat_rot.Rotation=&RigVMFunction_MathQuaternionFromEuler(Euler=(X=0,Y=0,Z=0), Euler.Y=@Multiply.Result).Result
   )
  ]}
 CollapseNode_ContainedGraph {nodes [
   RigVMFunctionEntryNode()
   Multiply = RigVMAggregateNode(
    C=&RigUnit_GetTransform(Item=(Type=Bone,Name="neck_01"), Space=LocalSpace, bInitial=False).Transform
    A=&RigUnit_GetTransform(Item=(Type=Bone,Name="head_end"), Space=LocalSpace, bInitial=False).Transform
    D=&RigUnit_GetTransform(Item=(Type=Bone,Name="neck_02"), Space=LocalSpace, bInitial=False).Transform
    B=&RigUnit_GetTransform(Item=(Type=Bone,Name="head"), Space=LocalSpace, bInitial=False).Transform
   )
   Unit = &RigVMFunction_MathVectorUnit(Value=@Multiply.Result.Translation) {template="Unit::Execute(in Value,out Result)"}
   RigVMFunctionReturnNode(Result=@Unit.Result)
  ]}
 Multiply_1_ContainedGraph {nodes [
   n1 = RigVMFunctionEntryNode()
   Multiply = &RigVMFunction_MathTransformMul(A=@n1.A, B=@n1.B) {template="Multiply::Execute(in A,in B,out Result)"}
   RigVMFunctionReturnNode(Result=&RigVMFunction_MathTransformMul(
     B=@n1.D
     A=&RigVMFunction_MathTransformMul(B=@n1.C, A=@Multiply.Result) {template="Multiply::Execute(in A,in B,out Result)"}.Result
    ) {template="Multiply::Execute(in A,in B,out Result)"}.Result)
  ]}
 Multiply_1_ContainedGraph {nodes [
   n1 = RigVMFunctionEntryNode()
   Multiply = &RigVMFunction_MathDoubleMul(A=@n1.A, B=@n1.B)
   RigVMFunctionReturnNode(Result=&RigVMFunction_MathDoubleMul(B=@n1.C, A=@Multiply.Result).Result)
  ]}
 RigVMFunctionLibrary {links [], nodes []}
 RigVMModel {nodes [
   RigUnit_BeginExecution = &RigUnit_BeginExecution()
   head_vector_nodes = RigVMCollapseNode()
   RigVMVariableNode(ExecuteContext=@RigUnit_BeginExecution.ExecutePin, Value=@head_vector_nodes.Result) {variable=head_vector}
   > RigVMVariableNode(Value=-80) {variable=rotate_collar}
   n1 = RigVMVariableNode() {variable=rotate_collar}
   n2 = RigVMCollapseNode(Bone=BASE_collar_r3_r, s_max=1, s_min=0.8, angle=@n1.Value)
   &RigUnit_ModifyTransforms(
    ExecutePin=()
    ItemToModify=((Item=(Type=Bone,Name="_A_A_A"),Transform=(Rotation=(X=0,Y=0,Z=0,W=1),Translation=(X=0,Y=0,Z=0),Scale3D=(X=1,Y=1,Z=1))))
    Weight=1
    ItemToModify.0.Transform.Rotation=@n2.quat_rot.Rotation
   )
   &RigUnit_ModifyTransforms(
    ExecutePin=()
    ItemToModify="((Item=(Type=Bone,Name=\"BASE_collar_l7_l\"),Transform=(Rotation=(X=0,Y=0,Z=0,W=1),Translation=(X=0,Y=0,Z=0),Scale3D=(X=1,Y=1,Z=1))),(Item=(Type=Bone,Name=\"collar_02\"),Transform=(Rotation=(X=0,Y=0,Z=0,W=1),Translation=(X=0,Y=0,Z=0),Scale3D=(X=1,Y=1,Z=1))),(I ...(truncated)"
    Weight=1
    WeightMinimum=0
    ItemToModify.7.Transform.Rotation=RigVMCollapseNode(Bone=BASE_collar_l2_l, s_max=1, s_min=0.8, angle=@n1.Value).quat_rot.Rotation
    ItemToModify.6.Transform.Rotation=RigVMCollapseNode(Bone=BASE_collar_l3_l, s_max=1, s_min=0.8, angle=@n1.Value).quat_rot.Rotation
    ItemToModify.9.Transform.Rotation=RigVMCollapseNode(Bone=BASE_collar_m, s_max=1, s_min=0.8, angle=@n1.Value).quat_rot.Rotation
    ItemToModify.11.Transform.Rotation=RigVMCollapseNode(Bone=BASE_collar_r2_r, s_max=1, s_min=0.8, angle=@n1.Value).quat_rot.Rotation
    ItemToModify.12.Transform.Rotation=@n2.quat_rot.Rotation
   )
   RigVMRerouteNode(Value=RigVMCollapseNode(Bone=BASE_collar_l1_l, s_max=1, s_min=0.8, angle=@n1.Value).quat_rot.Rotation)
   RigVMRerouteNode(Value=RigVMCollapseNode(Bone=BASE_collar_r2_r, s_max=1, s_min=0.8, angle=@n1.Value).quat_rot.Rotation)
  ]}
]
hierarchy [
 root {loc=<0 0 0>, rot=<0 0 0 1>}
 pelvis {parent=root, loc=<0 -0.7071 93.1191>, rot=<0.0127 0.707 -0.0127 -0.707>}
 spine_01 {parent=pelvis, loc=<6.7494 -0.6388 0>, rot=<0 0 0.034 -0.9994>}
 spine_02 {parent=spine_01, loc=<7.2518 -0.035 0>, rot=<0 0 0.0238 -0.9997>}
 spine_03 {parent=spine_02, loc=<6.0104 0 0>, rot=<0 0 0.0149 -0.9999>}
 spine_04 {parent=spine_03, loc=<8.1024 0 0>, rot=<0 0 -0.0132 -0.9999>}
 spine_05 {parent=spine_04, loc=<10.0146 0 0>, rot=<0 0 -0.0015 -1>}
 neck_01 {parent=spine_05, loc=<11.3138 0.1529 0>, rot=<0 0 -0.2003 -0.9797>}
 neck_02 {parent=neck_01, loc=<4.5627 0 0>, rot=<0 0 0.0504 -0.9987>}
 head {parent=neck_02, loc=<5.3914 0 0>, rot=<0 0 0.115 -0.9934>}
 BASE_dyn_L_hair_A_1 {parent=head, loc=<11.1353 -3.7054 -8.284>, rot=<0.1052 -0.9161 0.3186 -0.2197>}
 dyn_L_hair_A_1 {parent=BASE_dyn_L_hair_A_1, loc=<0 0 0>, rot=<0 0 0 1>}
 dyn_L_hair_A_2 {parent=dyn_L_hair_A_1, loc=<4.4166 0 0>, rot=<0 -0.2383 0 -0.9712>}
 dyn_L_hair_A_3 {parent=dyn_L_hair_A_2, loc=<5.0965 0 0>, rot=<0 0 0 1>}
 BASE_dyn_L_hair_B_1 {parent=head, loc=<8.1399 -6.7364 -5.3178>, rot=<0.0647 -0.8249 0.5509 -0.1086>}
 dyn_L_hair_B_1 {parent=BASE_dyn_L_hair_B_1, loc=<0 0 0>, rot=<0 0 0 1>}
 dyn_L_hair_B_2 {parent=dyn_L_hair_B_1, loc=<5.1504 0 0>, rot=<-0.0016 -0.1137 0.0143 -0.9934>}
 dyn_L_hair_B_3 {parent=dyn_L_hair_B_2, loc=<4.005 -1.0402 0.1324>, rot=<0 0 0 1>}
 dyn_L_hair_B_4 {parent=dyn_L_hair_B_2, loc=<4.005 2.2447 -1.0471>, rot=<0.1602 0 0 -0.9871>}
 BASE_dyn_C_hair_1 {parent=head, loc=<9.4109 -8.1807 0>, rot=<0.1372 -0.6937 0.6937 -0.1372>}
 dyn_C_hair_1 {parent=BASE_dyn_C_hair_1, loc=<0 0 0>, rot=<0 0 0 1>}
 dyn_C_hair_2 {parent=dyn_C_hair_1, loc=<5.2623 0 0>, rot=<0 -0.1604 0 -0.9871>}
 dyn_C_hair_3 {parent=dyn_C_hair_2, loc=<5.0826 0 0>, rot=<0 0 0 1>}
 BASE_dyn_R_hair_B_1 {parent=head, loc=<8.7612 -6.6361 5.5831>, rot=<0.1321 -0.5286 0.8334 -0.0924>}
 dyn_R_hair_B_1 {parent=BASE_dyn_R_hair_B_1, loc=<0 0 0>, rot=<0 0 0 1>}
 dyn_R_hair_B_2 {parent=dyn_R_hair_B_1, loc=<4.9692 0 0>, rot=<0 -0.1221 0 -0.9925>}
 dyn_R_hair_B_3 {parent=dyn_R_hair_B_2, loc=<4.8616 0 0>, rot=<0 0 0 1>}
 BASE_dyn_R_hair_A_1 {parent=head, loc=<9.6262 -3.8495 9.7729>, rot=<0.0766 -0.3142 0.9456 -0.036>}
 dyn_R_hair_A_1 {parent=BASE_dyn_R_hair_A_1, loc=<0 0 0>, rot=<0 0 0 1>}
 dyn_R_hair_A_2 {parent=dyn_R_hair_A_1, loc=<4.1508 0 0>, rot=<0 -0.1163 0 -0.9932>}
 dyn_R_hair_A_3 {parent=dyn_R_hair_A_2, loc=<6.3643 2.155 0>, rot=<0 0 0 1>}
 dyn_R_hair_A_4 {parent=dyn_R_hair_A_2, loc=<5.9791 -2.4584 -0.4675>, rot=<0 0 0 1>}
 head_end {parent=head, loc=<9.2548 0 0>, rot=<0 0 0 1>}
 dyn_frontLeaf {parent=head, loc=<17.8855 7.2678 -1.4785>, rot=<-0.033 0.6865 -0.2555 -0.6799>}
 dyn_frontLeaf_2 {parent=dyn_frontLeaf, loc=<3.4196 1.8625 0>, rot=<0.4771 0.1704 -0.0732 -0.8591>}
 dyn_frontLeaf_3 {parent=dyn_frontLeaf_2, loc=<3.582 0 0>, rot=<0 0 0 1>}
 dyn_frontLeaf_4 {parent=dyn_frontLeaf, loc=<4.2808 -1.3899 -0.1579>, rot=<-0.1498 -0.1129 0.1639 -0.9685>}
 dyn_frontLeaf_5 {parent=dyn_frontLeaf_4, loc=<2.9272 0 0>, rot=<0 0 0 1>}
 _A_A_A {parent=head, loc=<9.6262 -3.8495 9.7729>, rot=<0.0766 -0.3142 0.9456 -0.036>}
 clavicle_l {parent=spine_05, loc=<4.3575 -0.1466 -2.3223>, rot=<0.0465 -0.7289 -0.009 -0.6829>}
 upperarm_l {parent=clavicle_l, loc=<15.5111 0 0>, rot=<0.0226 -0.3234 -0.0088 -0.9459>}
 clavicle_r {parent=spine_05, loc=<4.3575 -0.1466 2.3223>, rot=<0.7289 0.0465 -0.6829 0.009>}
 upperarm_r {parent=clavicle_r, loc=<-15.5111 0 0>, rot=<0.0226 -0.3234 -0.0088 -0.9459>}
 BASE_collar_l3_l {parent=spine_05, loc=<10.3774 -2.6667 -7.3731>, rot=<0.0901 -0.5721 0.8149 0.0217>}
 BASE_collar_l2_l {parent=spine_05, loc=<10.1085 -4.6062 -5.8197>, rot=<0.04 -0.3145 0.9482 0.0218>}
 BASE_collar_l1_l {parent=spine_05, loc=<10.0267 -6.2033 -3.0844>, rot=<0.0377 -0.187 0.9814 0.0199>}
 BASE_collar_r1_r {parent=spine_05, loc=<10.0267 -6.2033 3.0844>, rot=<0.187 0.0377 0.0199 -0.9814>}
 BASE_collar_r2_r {parent=spine_05, loc=<10.1085 -4.6062 5.8197>, rot=<0.3144 0.0365 0.023 -0.9483>}
 BASE_collar_r3_r {parent=spine_05, loc=<10.3774 -2.6667 7.3731>, rot=<0.5226 0.0984 0.0119 -0.8468>}
 BASE_collar_m {parent=spine_05, loc=<9.991 -6.7419 0>, rot=<0.0002 -0.002 0.0198 -0.9998>}
]
variables [
 head_vector {type=FVector, default=(X=0,Y=0,Z=0)}
 rotate_collar {type=double, default=0}
]
