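# Physics asset for the LazyLizz Chip cosmetic face accessory, simulating hair strands and front-leaf pieces on the character's head.
# Five hair chains and a branching leaf chain hang from head-linked roots with tight swing and twist limits.
# Head, neck, upper spine, clavicle and upper-arm capsules form the body shapes, linked by free angular constraints.
# Extra free constraints cross-link neighboring hair chain ends.
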PhysicsAsset /BRCosmetics/Characters/Player/Male/Medium/Bodies/M_MED_LazyLizz_Chip/Meshes/Parts/M_MED_LazyLizz_Chip_FaceAcc_Physics
bodies {
 BASE_dyn_C_hair_1 [{shapeType=Capsule, length=0.505, radius=0.505, shapeName=BASE_dyn_C_hair_1_capsule, rotation=<rot 0 -90 0>}]
 BASE_dyn_L_hair_A_1 [{shapeType=Capsule, length=0.505, radius=0.505, shapeName=BASE_dyn_L_hair_A_1_capsule, rotation=<rot 0 -90 0>}]
 BASE_dyn_L_hair_B_1 [{shapeType=Capsule, length=0.505, radius=0.505, shapeName=BASE_dyn_L_hair_B_1_capsule, rotation=<rot 0 -90 0>}]
 BASE_dyn_R_hair_A_1 [{shapeType=Capsule, length=0.505, radius=0.505, shapeName=BASE_dyn_R_hair_A_1_capsule, rotation=<rot 0 -90 0>}]
 BASE_dyn_R_hair_B_1 [{shapeType=Capsule, length=0.505, radius=0.505, shapeName=BASE_dyn_R_hair_B_1_capsule, rotation=<rot 0 -90 0>}]
 _A_A_A [{shapeType=Capsule, length=7.3734, radius=4.2435, shapeName=_A_A_A_capsule, center=<11.894 0.1247 -9.3168>, rotation=<rot 88.0931 90 180>}]
 clavicle_l [{
   shapeType=Capsule
   length=6.4689
   radius=11.727
   shapeName=clavicle_l_capsule
   center=<6.6638 2.3303 -3.9854>
   rotation=<rot -0.0001 -90 107.8942>
  }]
 clavicle_r [{
   shapeType=Capsule
   length=6.4689
   radius=11.727
   shapeName=clavicle_r_capsule
   center=<-6.6638 -2.3303 3.9854>
   rotation=<rot -0.0001 -90 -72.1058>
  }]
 dyn_C_hair_1 [{shapeType=Capsule, length=3.7502, radius=2.9294, shapeName=dyn_C_hair_1_capsule, center=<2.2096 0 2.4338>, rotation=<rot 0 0 90>}]
 dyn_C_hair_2 [{shapeType=Capsule, length=3.2495, radius=2.9071, shapeName=dyn_C_hair_2_capsule, center=<2.2096 0 1.8814>, rotation=<rot 0 0 90>}]
 dyn_C_hair_3 [{shapeType=Capsule, length=3.0882, radius=2.7137, shapeName=dyn_C_hair_3_capsule, center=<3.3623 0 0.5562>, rotation=<rot 0 0 90>}]
 dyn_L_hair_A_1 [{shapeType=Capsule, length=1.979, radius=5.086, shapeName=dyn_L_hair_A_1_capsule, center=<2.2559 0 4.1847>, rotation=<rot 0 0 90>}]
 dyn_L_hair_A_2 [{shapeType=Capsule, length=3.593, radius=3.3078, shapeName=dyn_L_hair_A_2_capsule, center=<4.2566 0 2.4814>, rotation=<rot 0 0 90>}]
 dyn_L_hair_A_3 [{shapeType=Box, extentX=4.2686, extentY=7.4935, extentZ=2.374, shapeName=dyn_L_hair_A_3_box, center=<3.0509 0 -0.7518>}]
 dyn_L_hair_B_1 [{shapeType=Capsule, length=1.2127, radius=5.086, shapeName=dyn_L_hair_A_1_capsule, center=<2.2559 0 5.0787>, rotation=<rot 0 0 90>}]
 dyn_L_hair_B_2 [{shapeType=Capsule, length=3.593, radius=3.3078, shapeName=dyn_L_hair_A_2_capsule, center=<4.2566 0 2.4814>, rotation=<rot 0 0 90>}]
 dyn_L_hair_B_3 [{shapeType=Capsule, length=4.0429, radius=2.3143, shapeName=dyn_L_hair_A_3_capsule, center=<2.9514 0 0.2888>, rotation=<rot 0 0 90>}]
 dyn_L_hair_B_4 [{shapeType=Box, extentX=3.8633, extentY=4.4789, extentZ=2.0413, shapeName=dyn_L_hair_B_4_box, center=<2.3827 0 0>}]
 dyn_R_hair_A_1 [{shapeType=Box, extentX=4.4902, extentY=7.5, extentZ=2.2475, shapeName=dyn_R_hair_A_1_box, center=<2.9729 0 0.2913>}]
 dyn_R_hair_A_2 [{shapeType=Box, extentX=7.5, extentY=7.5, extentZ=3.0105, shapeName=dyn_R_hair_A_2_box, center=<4.3177 0 1.0168>}]
 dyn_R_hair_A_3 [{shapeType=Capsule, length=4.0429, radius=2.3143, shapeName=dyn_L_hair_A_3_capsule, center=<2.9514 0 -0.5867>, rotation=<rot 0 0 90>}]
 dyn_R_hair_A_4 [{shapeType=Capsule, length=4.4962, radius=1.861, shapeName=dyn_L_hair_A_3_capsule, center=<2.9514 0 -0.5867>, rotation=<rot 0 0 90>}]
 dyn_R_hair_B_1 [{shapeType=Capsule, length=3.5675, radius=2.7312, shapeName=dyn_L_hair_A_1_capsule, center=<2.7923 0 2.5191>, rotation=<rot 0 0 90>}]
 dyn_R_hair_B_2 [{shapeType=Capsule, length=3.593, radius=3.3078, shapeName=dyn_L_hair_A_2_capsule, center=<4.2566 0 2.339>, rotation=<rot 0 0 90>}]
 dyn_R_hair_B_3 [{shapeType=Capsule, length=4.0429, radius=2.3143, shapeName=dyn_L_hair_A_3_capsule, center=<2.9514 0 -0.5867>, rotation=<rot 0 0 90>}]
 dyn_frontLeaf [{
   shapeType=Capsule
   length=3.4438
   radius=3.3341
   shapeName=dyn_frontLeaf_capsule
   center=<1.1881 0.1437 0.4712>
   rotation=<rot 16.0459 -149.5306 -90>
  }]
 dyn_frontLeaf_2 [{
   shapeType=Capsule
   length=2.4166
   radius=1.853
   shapeName=dyn_frontLeaf_3_capsule
   center=<1.0703 0.9464 -0.0169>
   rotation=<rot 0 -82.1234 -3.5682>
  }]
 dyn_frontLeaf_3 [{
   shapeType=Capsule
   length=1.0661
   radius=1.853
   shapeName=dyn_frontLeaf_3_capsule
   center=<1.1526 1.1802 -0.4957>
   rotation=<rot 0 -82.1234 178.2772>
  }]
 dyn_frontLeaf_4 [{
   shapeType=Capsule
   length=2.0805
   radius=1.1306
   shapeName=dyn_frontLeaf_4_capsule
   center=<0.8947 0.1982 -0.5353>
   rotation=<rot 0 -163.2265 -82.3952>
  }]
 dyn_frontLeaf_5 [{
   shapeType=Capsule
   length=1.3565
   radius=1.1683
   shapeName=dyn_frontLeaf_5_capsule
   center=<0.3837 0.1116 -0.7655>
   rotation=<rot 8.394 -167.2264 -90>
  }]
 head [
  {shapeType=Capsule, length=1.3997, radius=11.6907, center=<3.1734 2.6019 0.0541>, rotation=<rot 0.5016 -90.415 -88.3249>}
  {
   shapeType=Capsule
   length=1.2035
   radius=8.2665
   shapeName=head_capsule
   center=<1.6133 -0.0375 -0.0239>
   rotation=<rot -89.1519 179.9981 -180>
  }
 ]
 neck_01 [{shapeType=Capsule, length=11.926, radius=8.5473, shapeName=neck_01_capsule, center=<5.8584 2.2424 0.0772>, rotation=<rot 0 -90 89.2454>}]
 spine_05 [{shapeType=Capsule, length=26.912, radius=12.436, shapeName=spine_05_capsule, center=<-3.3678 1.6151 0>, rotation=<rot 0 -90 0>}]
 upperarm_l [{shapeType=Capsule, length=8.9066, radius=5.4379, shapeName=upperarm_l_capsule, center=<6.0725 0 -3.2556>, rotation=<rot 0 -90 -90.3837>}]
 upperarm_r [{shapeType=Capsule, length=7.7997, radius=6.5447, shapeName=upperarm_r_capsule, center=<-4.8432 0 3.2638>, rotation=<rot 0 -90 89.6163>}]
}
constraints [
 {bone1Name=dyn_C_hair_1, bone2Name=BASE_dyn_C_hair_1, swing1LimitDegrees=10, swing2LimitDegrees=40, twistLimitDegrees=10}
 {bone1Name=dyn_C_hair_2, bone2Name=dyn_C_hair_1, swing1LimitDegrees=10, swing2LimitDegrees=40, twistLimitDegrees=10}
 {bone1Name=dyn_C_hair_3, bone2Name=dyn_C_hair_2, swing1LimitDegrees=10, swing2LimitDegrees=40, twistLimitDegrees=10}
 {bone1Name=dyn_L_hair_A_1, bone2Name=BASE_dyn_L_hair_A_1, swing1LimitDegrees=10, swing2LimitDegrees=40, twistLimitDegrees=10}
 {bone1Name=dyn_L_hair_A_2, bone2Name=dyn_L_hair_A_1, swing1LimitDegrees=10, swing2LimitDegrees=40, twistLimitDegrees=10}
 {bone1Name=dyn_L_hair_A_3, bone2Name=dyn_L_hair_A_2, swing1LimitDegrees=10, swing2LimitDegrees=40, twistLimitDegrees=10}
 {bone1Name=dyn_L_hair_B_1, bone2Name=BASE_dyn_L_hair_B_1, swing1LimitDegrees=10, swing2LimitDegrees=40, twistLimitDegrees=10}
 {bone1Name=dyn_L_hair_B_2, bone2Name=dyn_L_hair_B_1, swing1LimitDegrees=10, swing2LimitDegrees=40, twistLimitDegrees=10}
 {bone1Name=dyn_L_hair_B_3, bone2Name=dyn_L_hair_B_2, swing1LimitDegrees=10, swing2LimitDegrees=40, twistLimitDegrees=10}
 {bone1Name=dyn_L_hair_B_4, bone2Name=dyn_L_hair_B_2, swing1LimitDegrees=10, swing2LimitDegrees=40, twistLimitDegrees=10}
 {bone1Name=dyn_R_hair_A_1, bone2Name=BASE_dyn_R_hair_A_1, swing1LimitDegrees=10, swing2LimitDegrees=40, twistLimitDegrees=10}
 {bone1Name=dyn_R_hair_A_2, bone2Name=dyn_R_hair_A_1, swing1LimitDegrees=10, swing2LimitDegrees=40, twistLimitDegrees=10}
 {bone1Name=dyn_R_hair_A_3, bone2Name=dyn_R_hair_A_2, swing1LimitDegrees=10, swing2LimitDegrees=40, twistLimitDegrees=10}
 {bone1Name=dyn_R_hair_A_4, bone2Name=dyn_R_hair_A_2, swing1LimitDegrees=10, swing2LimitDegrees=40, twistLimitDegrees=10}
 {bone1Name=dyn_R_hair_B_1, bone2Name=BASE_dyn_R_hair_B_1, swing1LimitDegrees=10, swing2LimitDegrees=40, twistLimitDegrees=10}
 {bone1Name=dyn_R_hair_B_2, bone2Name=dyn_R_hair_B_1, swing1LimitDegrees=10, swing2LimitDegrees=40, twistLimitDegrees=10}
 {bone1Name=dyn_R_hair_B_3, bone2Name=dyn_R_hair_B_2, swing1LimitDegrees=10, swing2LimitDegrees=40, twistLimitDegrees=10}
 {bone1Name=dyn_frontLeaf_3, bone2Name=dyn_frontLeaf_2, swing1LimitDegrees=20, swing2LimitDegrees=10, twistLimitDegrees=5}
 {bone1Name=dyn_frontLeaf_5, bone2Name=dyn_frontLeaf_4, swing1LimitDegrees=5, swing2LimitDegrees=20, twistLimitDegrees=5}
 {
  bone1Name=BASE_dyn_C_hair_1
  bone2Name=head
  swing1LimitDegrees=45
  swing1Motion=Free
  swing2LimitDegrees=45
  swing2Motion=Free
  twistLimitDegrees=45
  twistMotion=Free
 }
 {
  bone1Name=BASE_dyn_L_hair_A_1
  bone2Name=head
  swing1LimitDegrees=45
  swing1Motion=Free
  swing2LimitDegrees=45
  swing2Motion=Free
  twistLimitDegrees=45
  twistMotion=Free
 }
 {
  bone1Name=BASE_dyn_L_hair_B_1
  bone2Name=head
  swing1LimitDegrees=45
  swing1Motion=Free
  swing2LimitDegrees=45
  swing2Motion=Free
  twistLimitDegrees=45
  twistMotion=Free
 }
 {
  bone1Name=BASE_dyn_R_hair_A_1
  bone2Name=head
  swing1LimitDegrees=45
  swing1Motion=Free
  swing2LimitDegrees=45
  swing2Motion=Free
  twistLimitDegrees=45
  twistMotion=Free
 }
 {
  bone1Name=BASE_dyn_R_hair_B_1
  bone2Name=head
  swing1LimitDegrees=45
  swing1Motion=Free
  swing2LimitDegrees=45
  swing2Motion=Free
  twistLimitDegrees=45
  twistMotion=Free
 }
 {
  bone1Name=upperarm_l
  bone2Name=clavicle_l
  swing1LimitDegrees=45
  swing1Motion=Free
  swing2LimitDegrees=45
  swing2Motion=Free
  twistLimitDegrees=45
  twistMotion=Free
 }
 {
  bone1Name=upperarm_r
  bone2Name=clavicle_r
  swing1LimitDegrees=45
  swing1Motion=Free
  swing2LimitDegrees=45
  swing2Motion=Free
  twistLimitDegrees=45
  twistMotion=Free
 }
 {
  bone1Name=head
  bone2Name=neck_01
  swing1LimitDegrees=45
  swing1Motion=Free
  swing2LimitDegrees=45
  swing2Motion=Free
  twistLimitDegrees=45
  twistMotion=Free
 }
 {
  bone1Name=clavicle_l
  bone2Name=spine_05
  swing1LimitDegrees=45
  swing1Motion=Free
  swing2LimitDegrees=45
  swing2Motion=Free
  twistLimitDegrees=45
  twistMotion=Free
 }
 {
  bone1Name=clavicle_r
  bone2Name=spine_05
  swing1LimitDegrees=45
  swing1Motion=Free
  swing2LimitDegrees=45
  swing2Motion=Free
  twistLimitDegrees=45
  twistMotion=Free
 }
 {bone1Name=dyn_frontLeaf_2, bone2Name=dyn_frontLeaf, swing1LimitDegrees=20, swing2LimitDegrees=10, twistLimitDegrees=5}
 {bone1Name=dyn_frontLeaf_4, bone2Name=dyn_frontLeaf, swing1LimitDegrees=5, swing2LimitDegrees=20, twistLimitDegrees=5}
 {
  bone1Name=dyn_frontLeaf
  bone2Name=head
  swing1LimitDegrees=45
  swing1Motion=Free
  swing2LimitDegrees=45
  swing2Motion=Free
  twistLimitDegrees=45
  twistMotion=Free
 }
 {
  bone1Name=_A_A_A
  bone2Name=head
  swing1LimitDegrees=45
  swing1Motion=Free
  swing2LimitDegrees=45
  swing2Motion=Free
  twistLimitDegrees=45
  twistMotion=Free
 }
 {
  bone1Name=dyn_L_hair_A_3
  bone2Name=dyn_L_hair_B_4
  swing1LimitDegrees=45
  swing1Motion=Free
  swing2LimitDegrees=45
  swing2Motion=Free
  twistLimitDegrees=45
  twistMotion=Free
 }
 {
  bone1Name=dyn_C_hair_3
  bone2Name=dyn_L_hair_B_3
  swing1LimitDegrees=45
  swing1Motion=Free
  swing2LimitDegrees=45
  swing2Motion=Free
  twistLimitDegrees=45
  twistMotion=Free
 }
 {
  bone1Name=dyn_C_hair_3
  bone2Name=dyn_R_hair_B_3
  swing1LimitDegrees=45
  swing1Motion=Free
  swing2LimitDegrees=45
  swing2Motion=Free
  twistLimitDegrees=45
  twistMotion=Free
 }
 {
  bone1Name=dyn_R_hair_B_3
  bone2Name=dyn_R_hair_A_3
  swing1LimitDegrees=45
  swing1Motion=Free
  swing2LimitDegrees=45
  swing2Motion=Free
  twistLimitDegrees=45
  twistMotion=Free
 }
 {
  bone1Name=neck_01
  bone2Name=spine_05
  swing1LimitDegrees=45
  swing1Motion=Free
  swing2LimitDegrees=45
  swing2Motion=Free
  twistLimitDegrees=45
  twistMotion=Free
 }
]
